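# Niagara system for the Season 42 backpack, adding glow, glitter and a reactive cube burst driven by User.KillCount.
# Burst_Cubes fires three spawn groups on kill-count changes gated by a comparison against zero, each group switching color, scale and motion.
# Coin_Glow001 loops socket-attached meshes and a light, with flickering alpha and time-of-day-scaled material intensity.
# Glitters spawns sparse meshes, animating mesh index over lifetime; front-end modifiers retune burst and glow.
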NiagaraSystem /BRCosmetics/Accessories/FORT_Backpacks/Backpack_M_MED_Season_42/FX/NS_Backpack_Season_42
dependencies {
 usedDataInterfaces [
  &NiagaraDataInterfaceMeshRendererInfo
  &NiagaraDataInterfaceCurve
  &NiagaraDataInterfaceArrayColor
  &NiagaraDataInterfaceSocketReader
  &NiagaraDataInterfaceSkeletalMesh
 ]
 usedDynamicInputs [&SetBoolByFloatComparison]
 usedModules [
  &self:SystemSpawnScript.SystemScriptSource.SystemScriptGraph.NiagaraNodeAssignment_1.SetVariables_#1
  &self:SystemSpawnScript.SystemScriptSource.SystemScriptGraph.NiagaraNodeAssignment_0.SetVariables_#2
  &SampleSkeletalMesh
  &SystemState
  /Game/Effects/Niagara/Modules/Logic/UpdateValueStateChange
 ]
}
emitters {
 Burst_Cubes {
  emitterUpdateScript [
   &EmitterState()
   /Niagara/Modules/Emitter/SpawnByTrigger(
    Spawn Count=5
    Trigger:NiagaraBool=&Boolean_AndOperation(
     Bool1=&SetBoolByIntComparison(A=$User.KillCount, B=0, Comparison Type=::NewEnumerator0)
     Bool2=&Boolean_AndOperation(Bool1=$System.UpdateValueStateChange.Changed, Bool2=&Boolean_NotOperation(Bool=$Transient.FirstFrame))
    )
    TriggerIntervalSwitch=false
   )
   &SpawnByTrigger(
    Spawn Count=7
    Spawn Group=1
    Trigger:NiagaraBool=&Boolean_AndOperation(
     Bool1=&SetBoolByIntComparison(A=$User.KillCount, B=0, Comparison Type=::NewEnumerator0)
     Bool2=&Boolean_AndOperation(Bool1=$System.UpdateValueStateChange.Changed, Bool2=&Boolean_NotOperation(Bool=$Transient.FirstFrame))
    )
    TriggerIntervalSwitch=false
   ) {module=SpawnByTrigger001}
   &SpawnByTrigger(
    Spawn Count=5
    Spawn Group=2
    Trigger:NiagaraBool=&Boolean_AndOperation(
     Bool1=&SetBoolByIntComparison(A=$User.KillCount, B=0, Comparison Type=::NewEnumerator0)
     Bool2=&Boolean_AndOperation(Bool1=$System.UpdateValueStateChange.Changed, Bool2=&Boolean_NotOperation(Bool=$Transient.FirstFrame))
    )
    TriggerIntervalSwitch=false
   ) {module=SpawnByTrigger002}
  ]
  particleSpawnScript [
   /SetVariables(Particles.VisibilityTag=&SpawnGroup()) {isSetParameters=true}
   &InitializeParticle(
    Color=&MergeHelper_ColorSwitch(
     0=&SelectLinearColorFromArray(
      Array Sampling Mode=::NewEnumerator0
      Direct Array Index=?
      Color Selection Array=DataInterface() {propertyValues {ColorData [<rgba 19 7.2256 0.0636 1>, <rgba 1 0.4403 0.8263 1>, <rgba 1 0.3585 0.2627 1>]}}
      Random Seed=0
     )
     1=<rgba 0.2562 0.9047 0.0686 1>
     2=&SelectLinearColorFromArray(
      Array Sampling Mode=::NewEnumerator0
      Direct Array Index=?
      Color Selection Array=DataInterface() {propertyValues {ColorData [<rgba 19 16.6528 0.5285 1>, <rgba 22 3.5051 0 1>]}}
      Random Seed=0
     )
     3=<rgba 1 1 1 1>
     4=<rgba 1 1 1 1>
     Index=&SpawnGroup()
     Num=::NewEnumerator3
    )
    Lifetime Max=1
    Lifetime Min=0.3
    Lifetime Mode=::NewEnumerator1
    Mesh Scale Mode=::NewEnumerator3
    Mesh Uniform Scale=&Multiply_Float(
     A=&MergeHelper_FloatSwitch(
      0=&RandomRangeFloat(Minimum=0.2, Maximum=0.3)
      1=&RandomRangeFloat(Minimum=0.2, Maximum=0.25)
      2=&RandomRangeFloat(Minimum=1, Maximum=1.1)
      3=0
      4=0
      Index=&SpawnGroup()
      Num=::NewEnumerator3
     )
     B=&FN_FrontEndFloatModifier(
      Gameplay=&MergeHelper_FloatSwitch(0=1, 1=1, 2=1, 3=?, Index=&SpawnGroup(), Num=::NewEnumerator1)
      Front End=&MergeHelper_FloatSwitch(0=1, 1=1, 2=1, 3=?, Index=&SpawnGroup(), Num=::NewEnumerator1)
     )
    )
    Sprite Size Mode=::NewEnumerator3
    Uniform Sprite Size=&FN_FrontEndFloatModifier(Gameplay=1, Front End=1)
    UsePositionOffset=true
    Write Material Random=true
   )
   &SocketLocation(
    SamplePosOffset=true
    Sampled Position Offset=&MergeHelper_VectorSwitch(0=<0 0 0>, 1=<0 0 0>, 2=?, 3=?, Index=&SpawnGroup(), Num=::NewEnumerator0)
    Socket Index=&MergeHelper_IntSwitch(0=0, 1=1, 2=?, 3=?, 4=?, 5=?, Index=&SpawnGroup(), Num=::NewEnumerator0)
    Socket Reader=$System.Socket Reader
    Socket Sampling Mode=::NewEnumerator2
    VelocitySpeedLimit=false
   )
   &AddVelocity(
    Velocity=&MergeHelper_VectorSwitch(
     0=&RandomRangeVector(Minimum=<50 40 0>, Maximum=<60 -40 0>)
     1=&RandomRangeVector(Minimum=<40 1 0>, Maximum=<-40 10 0>)
     2=&RandomRangeVector(Minimum=<50 40 0>, Maximum=<60 -40 0>)
     3=<0 0 0>
     4=<0 0 0>
     Index=&SpawnGroup()
     Num=::NewEnumerator3
    )
    Velocity Speed Scale=&MergeHelper_FloatSwitch(0=-3, 1=-3, 2=0, 3=0, 4=0, Index=&SpawnGroup(), Num=::NewEnumerator3)
   )
   &ShapeLocation(
    MaskSpawnGroup=false
    Sphere Distribution=::NewEnumerator3
    Sphere Radius=&FN_FrontEndFloatModifier(
     Gameplay=&MergeHelper_FloatSwitch(0=10, 1=5, 2=?, 3=?, Index=&SpawnGroup(), Num=::NewEnumerator0)
     Front End=&MergeHelper_FloatSwitch(0=5, 1=2, 2=?, 3=?, Index=&SpawnGroup(), Num=::NewEnumerator0)
    )
    UseOwnerScale=true
   )
   &DynamicMaterialParameters(
    Index 0 Param 1=&MergeHelper_FloatSwitch(
     0=&RandomRangeFloat(Minimum=0, Maximum=1)
     1=&RandomRangeFloat(Minimum=0, Maximum=1)
     2=0
     3=?
     Index=&SpawnGroup()
     Num=::NewEnumerator1
    )
    Index 0 Param 2=1
    Module.Index0Float0Write=true
    Module.Index0Float1Write=true
    Module.Index0Float2Write=true
    Module.Index0Float3Write=true
   )
  ]
  particleUpdateScript [
   &Drag(
    Drag=&FloatFromCurve(
     FloatCurve=DataInterface() {propertyValues {Curve {keys [{interpMode=RCIM_Cubic, leaveTangent=3.6589, tangentMode=RCTM_Break, time=0, value=0.15}, <at 0.8 1>]}}}
     Scale Curve=6
    )
    Use Linear Drag=true
    Use Rotational Drag=false
   )
   &GravityForce(Gravity=&Multiply_VectorByFloat(
     Vector=&MergeHelper_VectorSwitch(
      0=&RandomRangeVector(Minimum=<0 0 100>, Maximum=<0 0 200>)
      1=<0 0 0>
      2=&RandomRangeVector(Minimum=<0 0 100>, Maximum=<0 0 200>)
      3=?
      Index=&SpawnGroup()
      Num=::NewEnumerator1
     )
     Float=&FN_FrontEndFloatModifier(Gameplay=1, Front End=0.4)
    ))
   &VortexForce(
    UseOriginOffset=false
    Vortex Force Amount=&MergeHelper_FloatSwitch(
     0=&FloatFromCurve(
      FloatCurve=DataInterface() {propertyValues {Curve {keys [{arriveTangent=0.1055, interpMode=RCIM_Cubic, leaveTangent=0.1055, tangentMode=RCTM_User, time=0, value=1}]}}}
      Scale Curve=500
     )
     1=0
     2=?
     3=?
     Index=&SpawnGroup()
     Num=::NewEnumerator0
    )
   )
   &SolveForcesAndVelocity(Clamp Velocity=false, Limit Acceleration=false)
   &ParticleState()
   &UpdateMeshOrientation(
    Rotation Rate=&FloatFromCurve(FloatCurve=DataInterface() {propertyValues {Curve {keys [{arriveTangent=2.6546, interpMode=RCIM_Cubic, leaveTangent=2.6546, tangentMode=RCTM_User, time=0.2237, value=0}, <at 1 1>]}}})
    Rotation Vector=&RandomRangeVector(Minimum=<-1 -1 -1>, Maximum=<1 1 1>)
   )
   &ScaleMeshSize(Scale Factor=&VectorFromFloat(Value=&FloatFromCurve(FloatCurve=DataInterface() {propertyValues {Curve {keys [
          {interpMode=RCIM_Cubic, leaveTangent=5.173, tangentMode=RCTM_Break, time=0, value=0.5}
          {arriveTangent=1.0043, interpMode=RCIM_Cubic, leaveTangent=-0.3119, tangentMode=RCTM_Break, time=0.2, value=1}
          {arriveTangent=-2.3616, tangentMode=RCTM_User, time=1, value=0}
         ]}}})))
   &ScaleColor(
    Scale Alpha=&FloatFromCurve(FloatCurve=DataInterface() {propertyValues {Curve {keys [{interpMode=RCIM_Cubic, tangentMode=RCTM_Break, time=0.3, value=1}, <at 1 0>]}}})
    ScaleA=true
    ScaleRGB=true
   )
  ]
  rendererClasses [&NiagaraMeshRendererProperties]
  renderers [
   {rendererClass=&NiagaraMeshRendererProperties, rendererIndex=0}
   {rendererClass=&NiagaraMeshRendererProperties, rendererIndex=1}
   {rendererClass=&NiagaraMeshRendererProperties, rendererIndex=2}
  ]
 }
 Coin_Glow001 {
  emitterUpdateScript [
   &EmitterState(Life Cycle Mode=::NewEnumerator1, Loop Behavior=::NewEnumerator1, Loop Duration=1, UseLoopDelay=false)
   &SpawnBurst_Instantaneous(Use Spawn Probability=false, UseLoopCountLimit=false)
   &SpawnBurst_Instantaneous(Spawn Group=1, Use Spawn Probability=true, UseLoopCountLimit=false) {module=SpawnBurst_Instantaneous001}
   &SpawnBurst_Instantaneous(Spawn Group=2, Use Spawn Probability=true, UseLoopCountLimit=false) {module=SpawnBurst_Instantaneous002}
  ]
  particleSpawnScript [
   &InitializeParticle(
    Color=&MergeHelper_ColorSwitch(
     0=<rgba 19 8.528 0.6446 1>
     1=<rgba 19 8.528 0.6446 0.35>
     2=<rgba 0.2562 0.9047 0.0686 0.2>
     3=?
     Index=&SpawnGroup()
     Num=::NewEnumerator1
    )
    Lifetime=&MergeHelper_FloatSwitch(
     0=&RandomRangeFloat(Minimum=2, Maximum=5)
     1=&RandomRangeFloat(Minimum=1, Maximum=1)
     2=10
     3=?
     Index=&SpawnGroup()
     Num=::NewEnumerator1
    )
    Mesh Scale=&Multiply_VectorByFloat(Vector=<1 1.2 1>, Float=1)
    Mesh Scale Mode=::NewEnumerator4
    Sprite Size Mode=::NewEnumerator3
    Uniform Sprite Size=&MergeHelper_FloatSwitch(0=0, 1=0, 2=0, 3=?, Index=&SpawnGroup(), Num=::NewEnumerator1)
    UsePositionOffset=true
    Write Material Random=true
   )
   &InitialMeshOrientation(
    Mesh Orientation Mode=::NewEnumerator2
    Rotation=&MergeHelper_VectorSwitch(0=<0.25 0 0>, 1=<0.25 0 0>, 2=<0.5 0 0>, 3=<0 0 0>, 4=<0 0 0>, Index=&SpawnGroup(), Num=::NewEnumerator3)
    Use Rotation Vector=true
   )
   /SetVariables(Particles.VisibilityTag=&SpawnGroup()) {isSetParameters=true}
  ]
  particleUpdateScript [
   &ParticleState(Kill Particles When Lifetime Has Elapsed=false, Loop Particles Lifetime=true)
   &SocketLocation(
    SamplePosOffset=true
    Sampled Position Offset=&MergeHelper_VectorSwitch(0=<0 1.199 3.5>, 1=<0 1.199 3.5>, 2=<0 -7.1044 -18>, 3=?, Index=&SpawnGroup(), Num=::NewEnumerator1)
    Socket Index=&MergeHelper_IntSwitch(0=0, 1=0, 2=1, 3=?, 4=?, 5=?, Index=&SpawnGroup(), Num=::NewEnumerator1)
    Socket Reader=$System.Socket Reader
    Socket Sampling Mode=::NewEnumerator3
    VelocitySpeedLimit=false
   )
   &ScaleMeshSize(Scale Factor=&VectorFromFloat(Value=&FloatFromCurve(FloatCurve=DataInterface() {propertyValues {Curve {keys [{arriveTangent=-0.4286, interpMode=RCIM_Cubic, time=0.2, value=1}, {interpMode=RCIM_Cubic, leaveTangent=-0.25, time=1, value=1}]}}})))
   &DynamicMaterialParameters(
    Index 0 Param 1=&FN_FrontEndFloatModifier(
     Gameplay=&Multiply_Float(A=40, B=&FNTimeOfDayScalar(Night Scale=0.5, Morning Scale=1, Evening Scale=0.5))
     Front End=20
    )
    Module.Param0WriteEnabled=true
    Module.Param1WriteEnabled=true
    Module.Param2WriteEnabled=true
    Module.Param3WriteEnabled=true
   )
   &ScaleColor(
    Scale Alpha=&MergeHelper_FloatSwitch(
     0=&FloatFromCurve(FloatCurve=DataInterface() {propertyValues {Curve {keys [<at 0 1>,<at 1 1>]}}}, Scale Curve=0.07)
     1=&RandomRangeFloat(
      Minimum=&FloatFromCurve(
       FloatCurve=DataInterface() {propertyValues {Curve {keys [
           {interpMode=RCIM_Cubic, tangentMode=RCTM_User, time=0, value=0}
           {interpMode=RCIM_Cubic, tangentMode=RCTM_User, time=0.1953, value=1}
           <at 0.5038 1 RCIM_Cubic>
           {arriveTangent=-2.0153, interpMode=RCIM_Cubic, leaveTangent=-2.0153, time=0.8884, value=1}
           {interpMode=RCIM_Cubic, tangentMode=RCTM_User, time=1, value=0}
          ]}}}
       Scale Curve=0.07
      )
      Maximum=&FloatFromCurve(
       FloatCurve=DataInterface() {propertyValues {Curve {keys [
           {interpMode=RCIM_Cubic, tangentMode=RCTM_User, time=0, value=0}
           {interpMode=RCIM_Cubic, tangentMode=RCTM_User, time=0.5098, value=1}
           {interpMode=RCIM_Cubic, tangentMode=RCTM_User, time=1, value=0}
          ]}}}
       Scale Curve=0.08
      )
     )
     2=1
     3=?
     Index=&SpawnGroup()
     Num=::NewEnumerator1
    )
    Scale RGB=&VectorFromFloat(Value=4)
    ScaleA=true
    ScaleRGB=true
   )
   &FN_Light_Attributes(
    Color=<rgba 19 10.3365 0.3907 1>
    Create Color Attribute=true
    Create Exponent Attribute=true
    Create Position Attribute=true
    Create Radius Attribute=true
    Exponent=4
    Position=&AddVectorToPosition(Position=Linked() {linkedVariable {name=Output.SkeletalMeshLocation001.SampledPosition, type=Vector3f}}, Vector=<0 -5 0>)
    Radius=37
    Scale Color=0.15
   ) {module=FN_Light_Attributes001}
   /SetVariables(Particles.Cube_Color_01:LinearColor=<rgba 19 11.7924 0.0264 0.06>, Particles.Cube_Color_02:LinearColor=<rgba 19 9.6676 0.3587 0.04>) {isSetParameters=true}
   /SetVariables(Particles.1 Flicker Mesh INdex:NiagaraInt32=&RandomRangeInt(Minimum=0, Maximum=1)) {isSetParameters=true}
  ]
  rendererClasses [&NiagaraLightRendererProperties, &NiagaraMeshRendererProperties]
  renderers [
   {rendererClass=&NiagaraLightRendererProperties, rendererIndex=0}
   {rendererClass=&NiagaraMeshRendererProperties, rendererIndex=1}
   {rendererClass=&NiagaraMeshRendererProperties, rendererIndex=2}
   {rendererClass=&NiagaraMeshRendererProperties, rendererIndex=3}
   {rendererClass=&NiagaraMeshRendererProperties, rendererIndex=4}
  ]
 }
 Glitters {
  emitterSpawnScript [/SetVariables(Emitter.Emissiveness_Multiply=50) {isSetParameters=true}]
  emitterUpdateScript [
   &EmitterState()
   &SpawnRate(Spawn Probability=0.75, SpawnRate=2, Use Spawn Probability=true)
  ]
  particleSpawnScript [
   &InitializeParticle(
    Lifetime=&RandomRangeFloat(Minimum=0.3, Maximum=1)
    Mesh Renderer Array Visibility Mode=::NewEnumerator1
    Mesh Renderer Info=DataInterface() {propertyValues {MeshRenderer=&self:Glitters_0.NiagaraMeshRendererProperties_0}}
    Mesh Renderer Mesh Index:NiagaraInt32=$Particles.MeshIndex
    Mesh Scale Mode=::NewEnumerator3
    Mesh Uniform Scale=&RandomRangeFloat(Minimum=0.8, Maximum=1.2)
    Sprite Size Mode=::NewEnumerator3
    Uniform Sprite Size=30
    UsePositionOffset=false
    Write Material Random=true
   )
   &OffsetPosition(Position Offset=<5 -10 0>)
   &ShapeLocation(
    Handle Radius=3.1101
    Large Radius=14
    MaskSpawnGroup=false
    Override Seed=false
    Rotation Angle=2.6384
    Rotation Axis=<0 1 0>
    Rotation Mode=::NewEnumerator1
    Shape Primitive=::NewEnumerator3
    UseOwnerScale=true
   )
   &InitialMeshOrientation(
    Mesh Orientation Mode=::NewEnumerator7
    Rotation=&RandomRangeVector(Minimum=<0 0 0>, Maximum=<0 1 0>)
    Use Rotation Vector=true
   )
  ]
  particleUpdateScript [
   &ParticleState()
   &ScaleColor(
    Scale Alpha=&FloatFromCurve(
     FloatCurve=DataInterface() {propertyValues {Curve {keys [
         {interpMode=RCIM_Cubic, tangentMode=RCTM_User, time=0, value=0}
         {
          arriveTangent=-0.1193
          arriveTangentWeight=0.5115
          interpMode=RCIM_Cubic
          leaveTangent=-0.1193
          leaveTangentWeight=0.5115
          tangentMode=RCTM_User
          tangentWeightMode=RCTWM_WeightedBoth
          time=0.5014
          value=1
         }
         {interpMode=RCIM_Cubic, tangentMode=RCTM_User, time=1, value=0}
        ]}}}
     Scale Curve=&RandomRangeFloat(Minimum=1, Maximum=1)
    )
    ScaleA=true
    ScaleRGB=true
   )
   /SetVariables(Particles.MeshIndex=&MakeIntFromFloat(Float=&FloatFromCurve(
      FloatCurve=DataInterface() {propertyValues {Curve {keys [<at 0.0152 -1>,<at 0.5 5>,<at 0.981 -1>]}}}
      Scale Curve=&RandomRangeFloat(Minimum=0.8, Maximum=1)
     ))) {isSetParameters=true}
  ]
  rendererClasses [&NiagaraMeshRendererProperties]
  renderers [
   {rendererClass=&NiagaraMeshRendererProperties, rendererIndex=0}
   {rendererClass=&NiagaraMeshRendererProperties, rendererIndex=1}
  ]
 }
}
summary {
 systemName=NS_Backpack_Season_42
 emitters [
  {emitterName=Glitters, rendererClasses [&NiagaraMeshRendererProperties]}
  {emitterName=Burst_Cubes, rendererClasses [&NiagaraMeshRendererProperties]}
  {emitterName=Coin_Glow001, rendererClasses [&NiagaraLightRendererProperties, &NiagaraMeshRendererProperties]}
 ]
 userVariables [
  User.bIsFrontEnd {type=NiagaraBool}
  User.bIsFrontEndPreview {type=NiagaraBool}
  User.IsReactive {type=NiagaraFloat, defaultValue=1, description=""}
  User.KillCount {type=NiagaraInt32}
 ]
}
