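# Material function for a Season 42 backpack cosmetic that splits a 0 to 99 Number input into two per-digit phase values.
# First Position Phase comes from the tens digit (floor of Number / 10) and Second Position Phase from the ones digit (floor of fmod by 10).
# Each digit is scaled by 0.1 and saturated.
# The output descriptions appear swapped relative to this math.
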MaterialFunction /BRCosmetics/Accessories/FORT_Backpacks/Backpack_M_MED_Season_42/Materials/MF_TwoDigitNumberPhase
graph {nodes [
  n1 = FunctionInput() {
   bUsePreviewValueAsDefault=true
   description=A number in the range 0 to 99.
   inputName=Number
   inputType=FunctionInput_Scalar
   sortPriority=10
  }
  n2 = Add(B=@n1)
  n3 = Floor(None=Divide(A=@n2) {constB=10})
  FunctionOutput(None=Saturate(None=Multiply(A=@n3) {constB=0.1})) {description=The value from 0 to 1 representing the numbers 0 to 9., outputName=First Position Phase}
  n4 = Floor(None=Fmod(A=@n2, B=Constant() {r=10}))
  FunctionOutput(None=Saturate(None=Multiply(A=@n4) {constB=0.1})) {description=The value from 0 to 1 representing the numbers 10 to 90., outputName=Second Position Phase, sortPriority=10}
 ]}
