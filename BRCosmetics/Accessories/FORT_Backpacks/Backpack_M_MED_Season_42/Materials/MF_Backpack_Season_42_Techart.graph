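# Material function for the Season 42 backpack that shows a two-digit number, driven by a clamped KillCount parameter, by masking mesh elements through Opacity Mask.
# MF_TwoDigitNumberPhase yields a phase per digit, which sphere masks compare against red and green vertex color to select visible geometry.
# Local X picks which digit mask applies, and a local Z cutoff limits the effect, keeping the incoming mask elsewhere.
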MaterialFunction /BRCosmetics/Accessories/FORT_Backpacks/Backpack_M_MED_Season_42/Materials/MF_Backpack_Season_42_Techart
graph {nodes [
  Constant()
  n1 = GetMaterialAttributes(None=FunctionInput() {inputName=Material IN, inputType=FunctionInput_MaterialAttributes}) {attributeGetTypes [#1, #2, #3]}
  n2 = LocalPosition()
  n3 = ScalarParameter() {group=01_Season42, parameterName=NumberOpacityZCutoff}
  n4 = ScalarParameter() {defaultValue=0.025, group=03_Opacity Mask, parameterName=Global Character Vertex Hide Radius}
  n5 = Constant() {r=1}
  n6 = Constant() {r=1}
  n7 = MaterialFunctionCall!(Number=Add(B=NamedRerouteUsage())) {materialFunction=&/MF_TwoDigitNumberPhase}
  n8 = Constant() {r=1}
  n9 = LocalPosition()
  n10 = LinearInterpolate(
   A=SphereMask(
    A=If(A=@n7.First Position Phase, "A > B"=@n7.First Position Phase, "A == B"=@n7.First Position Phase, "A < B"=@n6) {constB=0.1}
    B=VertexColor().R
    Radius=@n4
    Hardness=@n5
   )
   B=SphereMask(
    A=If(A=@n7.Second Position Phase, "A > B"=@n7.Second Position Phase, "A == B"=@n7.Second Position Phase, "A < B"=@n8) {constB=0.1}
    B=VertexColor().G
    Radius=@n4
    Hardness=@n5
   )
   Alpha=Step(Y=ComponentMask(None=@n9.XYZ) {r=true}, X=Constant()) {constY=-1}
  )
  n11 = Saturate(None=Ceil(None=Constant() {desc=Hide Element Toggle, r=1}))
  n12 = SetMaterialAttributes!(
   MaterialAttributes=@n1.MaterialAttributes
   EmissiveColor
   Opacity
   OpacityMask=LinearInterpolate(A=@n1.OpacityMask, B=Multiply(A=@n10, B=@n11), Alpha=Step(Y=@n2.Z, X=@n3) {constY=-1})
   BaseColor
  ) {attributeSetTypes [
    #1
    #4
    #3
    #2
   ]}
  FunctionOutput(None=@n12) {outputName=Material OUT}
  NamedRerouteDeclaration(None=Clamp(None=ScalarParameter() {parameterName=KillCount}) {maxDefault=99}) {name=NumberA, nodeColor=<rgba 0 0.0187 1 1>}
  ScalarParameter() {group=01_Season42, parameterName=TestNumber, sliderMax=99}
 ]}
