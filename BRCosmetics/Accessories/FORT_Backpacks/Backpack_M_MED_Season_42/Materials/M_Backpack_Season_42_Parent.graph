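# Masked parent material for the Season 42 backpack cosmetic, built for skeletal meshes with morph targets and mesh deformers.
# It wraps the shared 2023 character parent functions around a backpack-specific function, MF_Backpack_Season_42_Techart.
# PreFX output passes through that Techart function as material attributes into PostFX, which also receives PreFX auxiliary outputs directly.
# The shading model comes from the graph, spanning lit, subsurface, clear coat and cloth models.
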Material /BRCosmetics/Accessories/FORT_Backpacks/Backpack_M_MED_Season_42/Materials/M_Backpack_Season_42_Parent
graph {
 properties {
  bAutomaticallySetUsageInEditor=false
  bUseMaterialAttributes=true
  bUsedWithMeshDeformer=true
  bUsedWithMorphTargets=true
  bUsedWithSkeletalMesh=true
  blendMode=BLEND_Masked
  pixelDepthOffsetMode=PDOM_Legacy
  previewMesh=/BRCosmetics/Weapons/FORT_Melee/Pickaxe_Season_42/Demo/Season_42
  shadingModel=MSM_FromMaterialExpression
  usedShadingModels=Default Lit | Subsurface | Clear Coat | Subsurface Profile | Cloth
 }
 nodes [
  n1 = MaterialFunctionCall() {materialFunction=&MF_CharacterParent_2023_PreFX}
  MaterialFunctionCall(
   Material=MaterialFunctionCall!(Material IN=@n1.Result) {materialFunction=&/MF_Backpack_Season_42_Techart}.Material OUT
   M=@n1.M
   Pre Skinned Local Position=@n1.M
   Distance Blend=@n1.M
   Fresnel=@n1.M
  ) {materialFunction=&MF_CharacterParent_2023_PostFX}
 ]
}
