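# Animation Blueprint for the Season 42 backpack cosmetic, chaining additive bone modifiers on the pack and its dyn_coin_01 bone.
# Each update computes a time-driven coin spin and sine bob, a skydiving value, and pack offsets from owner spine twist.
# Enemy elimination events advance a clamped stage, though every stage fires the same brief isEnemyPlayerEliminated pulse.
# In the front end, that pulse repeats on a delay loop.
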AnimBlueprint /BRCosmetics/Accessories/FORT_Backpacks/Backpack_M_MED_Season_42/Meshes/M_MED_Season_42_Pack_AnimBP
target_skeleton=&/M_MED_Season_42_Pack_Skeleton
graphs [
 AnimGraph {nodes [
   n1 = "Transform (Modify) Bone - Bone: pack"(ComponentPose=Local To Component(LocalPose=Input Pose()), Translation="0,-0.5,0", Rotation="0,0,5") {rotationMode=BMM_Additive, translationMode=BMM_Additive, boneToModify {boneName=pack}, rotation {roll=5}, translation {y=-0.5}}
   n2 = "Transform (Modify) Bone - Bone: pack"(ComponentPose=@n1, Translation="0,-3,0", Rotation="-0.000,0.000,5.490") {
    rotationMode=BMM_Additive
    translationMode=BMM_Additive
    alphaScaleBiasClamp {bInterpResult=true}
    boneToModify {boneName=pack}
    rotation {roll=5.49}
    translation {y=-3}
   }
   n3 = "Transform (Modify) Bone - Bone: dyn_coin_01"(ComponentPose=@n2, Alpha=1, Translation="0,0,0", Rotation="0,0,0", Scale="1,1,1") {rotationMode=BMM_Additive, translationMode=BMM_Additive, alphaScaleBiasClamp {bInterpResult=true}, boneToModify {boneName=dyn_coin_01}}
   Output Pose(Result=Component To Local(ComponentPose=@n3)) {layerGroup=None}
  ]}
 EventGraph {nodes [
   Event BlueprintUpdateAnimation()
   > n1 = Sequence!(execute) -> (then_0, then_1, then_2:Exec)
   n2 = "Sin (Radians)"(A=float * float(A=GetGameTimeInSeconds(), B=25))
   > Set translateCoin(translateCoin:Vector=MakeVector(Y=float + float(A=@n2, B=-15) {is=title K2Node_CommutativeAssociativeBinaryOperator})) -> (Output_Get:Vector)
   > Set Skydiving(Skydiving:double="To Float (Boolean)"(InBool=Get bIsSkyDiving())) -> (Output_Get:double)
   n3 = float * float(A=GetGameTimeInSeconds(), B=800)
   Set spincoin(execute=@n1.then_1, spincoin:Rotator=MakeRotator(Yaw=float * float(A=@n3, B=1) {is=title K2Node_PromotableOperator})) -> (Output_Get:Rotator)
   n4 = GetOwnerSkeletalMeshComponent()
   n5 = BreakRotator(InRot=BreakTransform(InTransform=GetDeltaTransformFromRefPose(self=@n4, BoneName=spine_05, BaseName=spine_01)).Rotation)
   Set backpack_rotate(execute=@n1.then_2, backpack_rotate=MakeRotator(Roll=MapRangeClamped(Value=@n5.Yaw, InRangeA=0, InRangeB=-20, OutRangeB=5)))
   > Set backpack_translation(backpack_translation=MakeVector(Y=MapRangeClamped(Value=@n5.Yaw, InRangeA=0, InRangeB=-30, OutRangeA=-0.2, OutRangeB=2)))
   n6 = Event BlueprintProcessAnimInputEvent()
   > Branch(Condition=IsNamed(self=@n6.AnimInputEvent, InEventName=aie.EnemyPlayerEliminated))
   > n7 = Increment Int(Value=Get EliminatedStage() -> (EliminatedStage:int=0))
   > Set EliminatedStage(EliminatedStage:int="Clamp (Integer)"(Value=@n7, Min=Get EliminatedStage() -> (EliminatedStage=0), Max=Get MaxEliminatedStage() -> (MaxEliminatedStage:int=0)) -> (ReturnValue=0)) -> (Output_Get:int=0)
   > n8 = Switch on Int!(execute:Exec, Selection:int=Get EliminatedStage() -> (EliminatedStage=0)) -> (Default:Exec, 0:Exec, 1:Exec, 2:Exec, 3:Exec, 4:Exec)
   Set isEnemyPlayerEliminated(execute=[@n8.0, @n8.Default], isEnemyPlayerEliminated:double=1) -> (Output_Get:double)
   > Delay(Duration=0.9)
   > Set isEnemyPlayerEliminated(isEnemyPlayerEliminated=0)
   Set isEnemyPlayerEliminated(execute=@n8.1, isEnemyPlayerEliminated=1)
   > Delay(Duration=0.9)
   > Set isEnemyPlayerEliminated(isEnemyPlayerEliminated=0)
   Set isEnemyPlayerEliminated(execute=@n8.2, isEnemyPlayerEliminated=1)
   > Delay(Duration=0.9)
   > Set isEnemyPlayerEliminated(isEnemyPlayerEliminated=0)
   Set isEnemyPlayerEliminated(execute=@n8.3, isEnemyPlayerEliminated=1)
   > Delay(Duration=0.9)
   > Set isEnemyPlayerEliminated(isEnemyPlayerEliminated=0)
   Set isEnemyPlayerEliminated(execute=@n8.4, isEnemyPlayerEliminated=1)
   > Delay(Duration=0.9)
   > Set isEnemyPlayerEliminated(isEnemyPlayerEliminated=0)
   Event OnCharacterCustomizationLoaded()
   > Branch(Condition=Get bIsInFrontEnd())
   > n9 = Delay(Duration=1)
   Set isEnemyPlayerEliminated(execute=[@n9.then, @n10.then], isEnemyPlayerEliminated=1)
   > Delay(Duration=0.9)
   > Set isEnemyPlayerEliminated(isEnemyPlayerEliminated=0)
   > n10 = Delay(Duration=7.1)
   TryGetPawnOwner()
   MapRangeClamped(Value=0, InRangeA=0, InRangeB=-30, OutRangeB=30)
  ]}
]
variables [
 backpack_rotate {type=Rotator}
 backpack_translation {type=Vector}
 EliminatedStage {type=integer}
 isEnemyPlayerEliminated {type=number}
 MaxEliminatedStage {type=integer}
 Skydiving {type=number}
 spincoin {type=Rotator}
 translateCoin {type=Vector}
]
